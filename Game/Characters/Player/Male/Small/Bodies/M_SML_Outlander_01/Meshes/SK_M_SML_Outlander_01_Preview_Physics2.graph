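# Physics asset for the preview skeletal mesh of the small male Outlander 01 player body, providing simple collision volumes for the upper body.
# Capsules cover pelvis, spine, clavicles and neck, with extra capsules layered on spine_02 and spine_05, and a sphere for the head.
# Arm, leg and foot bodies exist but carry no shapes.
# No constraints link the bodies, so it defines no jointed ragdoll.
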PhysicsAsset /Game/Characters/Player/Male/Small/Bodies/M_SML_Outlander_01/Meshes/SK_M_SML_Outlander_01_Preview_Physics2
bodies {
 calf_l []
 calf_r []
 clavicle_l [{shapeType=Capsule, length=7.7636, radius=11.1427, center=<9.9642 -0.7741 -1.5923>, rotation=<rot 68.0107 1.6353 177.121>}]
 clavicle_r [{shapeType=Capsule, length=7.7636, radius=11.1427, center=<-10.0313 0.7716 1.4601>, rotation=<rot 71.0843 1.1575 -3.3269>}]
 foot_l []
 foot_r []
 head [{shapeType=Sphere, radius=12.8478, center=<4.6388 2.3794 0>}]
 lowerarm_l []
 lowerarm_r []
 neck_02 [{shapeType=Capsule, length=13.1358, radius=12.1217, center=<-0.3962 1.4228 0>, rotation=<rot 90 -179.5524 176.1691>}]
 pelvis [{shapeType=Capsule, length=8.6562, radius=13.5388, center=<-2.6215 -1.3695 0>}]
 spine_01 [{shapeType=Capsule, length=10.7187, radius=12.4138, center=<-2.6215 -0.7156 0>}]
 spine_02 [
  {shapeType=Capsule, length=8.9329, radius=12.8245, center=<-2.6215 0.4993 0.0001>}
  {shapeType=Capsule, length=29.875, radius=11.1013, center=<2.5024 0.4992 3.9825>, rotation=<rot 90 0.0049 0.0049>}
  {shapeType=Capsule, length=29.875, radius=11.1013, center=<2.4704 0.4991 -3.9825>, rotation=<rot 90 -0.0396 -0.0396>}
 ]
 spine_03 [{shapeType=Capsule, length=10.2476, radius=10.7575, center=<-1.953 0.1364 0.0001>}]
 spine_04 [{shapeType=Capsule, length=6.7884, radius=13.6325, center=<2.1467 -1.2748 0.0002>}]
 spine_05 [
  {shapeType=Capsule, length=10.404, radius=14.019, center=<0.7944 2.4388 0>}
  {shapeType=Capsule, length=8.1336, radius=16.2028, center=<-0.3292 1.2157 0>}
 ]
 thigh_l []
 thigh_r []
 upperarm_l []
 upperarm_r []
}
constraints []
